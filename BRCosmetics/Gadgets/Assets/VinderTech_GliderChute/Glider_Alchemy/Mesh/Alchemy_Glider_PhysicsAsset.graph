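# Physics asset for the Alchemy glider cosmetic, a winged creature skeleton with head, jaws, feathered wings, tail wings and clawed legs.
# Capsule bodies cover the individual bones, while Body_Offset also carries a large sphere that is far bigger than any other shape.
# Constraints branch from Body_Offset through neck, wings, tail and legs, all sharing identical swing and twist limit values.
# Claw coverage is asymmetric, with two bodies on the left and one on the right.
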PhysicsAsset /BRCosmetics/Gadgets/Assets/VinderTech_GliderChute/Glider_Alchemy/Mesh/Alchemy_Glider_PhysicsAsset
bodies {
 Body_Offset [
  {shapeType=Sphere, radius=282, shapeName=Body_Offset_sphere}
  {shapeType=Capsule, length=45.0715, radius=34.5093, center=<0 -9.0617 27.413>, rotation=<rot 20.4398 -90 0>}
 ]
 Calf_l [{shapeType=Capsule, length=28.8869, radius=10.4067, center=<-0.7454 -29.0006 0.3014>, rotation=<rot 0 -4.0122 -89.1537>}]
 Calf_r [{shapeType=Capsule, length=31.2258, radius=10.7063, center=<0.8855 31.2797 -0.2653>, rotation=<rot 0 176.0501 -89.8913>}]
 Claw_front_l_m_1 [{shapeType=Capsule, length=14.1971, radius=10.462, center=<11.087 -0.093 -6.6537>, rotation=<rot 0 -90.7956 -50.1597>}]
 Claw_front_l_r_1 [{shapeType=Capsule, length=13.4608, radius=10.8331, center=<11.166 -0.2418 -5.2133>, rotation=<rot 0 -90.8297 -55.1024>}]
 Claw_front_r_1 [{shapeType=Capsule, length=12.7384, radius=11.2654, center=<-10.4448 -1.0153 5.0559>, rotation=<rot 0 -87.694 -53.2723>}]
 Feather_1_l [{shapeType=Capsule, length=39.6997, radius=10.1942, center=<37.1391 0.9249 0.1795>, rotation=<rot 0 -88.7875 -83.756>}]
 Feather_1_r [{shapeType=Capsule, length=39.6997, radius=10.1821, center=<-37.6029 -0.3848 3.7141>, rotation=<rot 0 -89.3247 -79.2347>}]
 Feather_2_l [{shapeType=Capsule, length=42.5822, radius=9.5261, center=<39.3113 0.0793 2.1042>, rotation=<rot 0 -90.373 -86.4761>}]
 Feather_2_r [{shapeType=Capsule, length=42.5822, radius=9.515, center=<-39.9245 0.113 1.9701>, rotation=<rot 0 -90.4583 -81.9192>}]
 Feather_3_l [{shapeType=Capsule, length=38.2928, radius=10.8021, center=<33.6185 -1.8051 6.2693>, rotation=<rot 0 87.1194 -78.9447>}]
 Feather_3_r [{shapeType=Capsule, length=38.2928, radius=10.7878, center=<-35.066 1.3046 -3.2975>, rotation=<rot 0 87.8592 -83.544>}]
 Head [{shapeType=Capsule, length=60.7103, radius=29.7285, center=<49.0229 5.407 0>, rotation=<rot 0 91.9097 -90>}]
 Jaw [{shapeType=Capsule, length=49.1666, radius=8.5961, center=<44.2598 -1.1358 0.0002>, rotation=<rot 0 -90.1759 -89.9996>}]
 Jaw_Left1 [{shapeType=Capsule, length=23.1584, radius=9.0019, center=<18.5877 -0.9064 2.1685>, rotation=<rot 0 89.4759 -79.9569>}]
 Jaw_Right1 [{shapeType=Capsule, length=23.1593, radius=9.0017, center=<-18.5872 0.9112 -2.1627>, rotation=<rot 0 89.4638 -79.9327>}]
 Neck_01 [{shapeType=Capsule, length=34.9831, radius=25.8577, center=<-3.383 5.8685 0>, rotation=<rot 0 110.6159 -90>}]
 Neck_Armor_l [{shapeType=Capsule, length=18.0006, radius=9.3394, center=<6.1697 -3.1795 2.3623>, rotation=<rot 0 -141.3843 -81.3782>}]
 Neck_Armor_r [{shapeType=Capsule, length=18.0006, radius=9.3394, center=<-6.1697 3.1795 -2.3623>, rotation=<rot 0 -141.3843 -81.3782>}]
 Tail [{shapeType=Capsule, length=30.8323, radius=24.815, center=<0.8437 -1.3932 -2.3123>, rotation=<rot -1.8086 -90 5.0675>}]
 Tail1 [{shapeType=Capsule, length=34.684, radius=14.1, center=<-29.4086 -4.6827 0.0455>, rotation=<rot 0 -87.6324 -89.9194>}]
 Tail_wing_l_1 [{shapeType=Capsule, length=42.1292, radius=19.726, center=<30.8489 -3.2664 8.3312>, rotation=<rot 0 87.0518 -70.8964>}]
 Tail_wing_l_2 [{shapeType=Capsule, length=74.2208, radius=31.7605, center=<73.6453 11.5421 -6.6861>, rotation=<rot 0 -82.2663 -65.3042>}]
 Tail_wing_r_1 [{shapeType=Capsule, length=42.1292, radius=19.726, center=<-30.8435 3.2657 -8.3272>, rotation=<rot 0 87.0518 -70.8965>}]
 Tail_wing_r_2 [{shapeType=Capsule, length=74.2208, radius=31.7605, center=<-73.6405 -11.5421 6.6908>, rotation=<rot 0 -82.2664 -65.3041>}]
 Thigh_l [{shapeType=Capsule, length=29.7138, radius=20.3658, center=<12.9759 -2.5169 12.6321>, rotation=<rot 0 97.549 -86.7336>}]
 Thigh_r [{shapeType=Capsule, length=29.7138, radius=20.3658, center=<-12.9759 2.5102 -12.6321>, rotation=<rot 0 97.549 -86.7336>}]
 Wing_shoulder_l [{shapeType=Capsule, length=25.4459, radius=25.1306, center=<17.9619 -3.3484 -5.2355>, rotation=<rot 0 87.7625 -62.4507>}]
 Wing_shoulder_l1 [{shapeType=Capsule, length=51.3046, radius=38.2719, center=<44.7197 -3.4538 -28.1272>, rotation=<rot 0 -93.069 -79.2524>}]
 Wing_shoulder_l2 [{shapeType=Capsule, length=43.2714, radius=30.3047, center=<37.9514 0.6236 -8.5235>, rotation=<rot 0 90.4499 -88.3629>}]
 Wing_shoulder_l3 [{shapeType=Capsule, length=40.7077, radius=13.0132, center=<37.4757 0.5141 -3.209>, rotation=<rot 0 -88.1185 -85.6496>}]
 Wing_shoulder_r [{shapeType=Capsule, length=25.4459, radius=25.1306, center=<-17.9619 3.3484 5.2355>, rotation=<rot 0 87.7625 -62.4508>}]
 Wing_shoulder_r1 [{shapeType=Capsule, length=51.3046, radius=38.2987, center=<-46.7306 2.8794 24.7304>, rotation=<rot 0 -92.4302 -83.5302>}]
 Wing_shoulder_r2 [{shapeType=Capsule, length=43.2714, radius=30.316, center=<-36.7118 -0.7255 10.8357>, rotation=<rot 0 -89.6988 -87.2212>}]
 Wing_shoulder_r3 [{shapeType=Capsule, length=40.7077, radius=12.7039, center=<-109.4483 -19.7463 1.5037>, rotation=<rot 0 120.2883 -86.8465>}]
}
constraints [
 {bone1Name=Neck_01, bone2Name=Body_Offset, swing1LimitDegrees=45, swing2LimitDegrees=45, twistLimitDegrees=45}
 {bone1Name=Neck_Armor_l, bone2Name=Neck_01, swing1LimitDegrees=45, swing2LimitDegrees=45, twistLimitDegrees=45}
 {bone1Name=Head, bone2Name=Neck_01, swing1LimitDegrees=45, swing2LimitDegrees=45, twistLimitDegrees=45}
 {bone1Name=Jaw, bone2Name=Head, swing1LimitDegrees=45, swing2LimitDegrees=45, twistLimitDegrees=45}
 {bone1Name=Jaw_Left1, bone2Name=Head, swing1LimitDegrees=45, swing2LimitDegrees=45, twistLimitDegrees=45}
 {bone1Name=Jaw_Right1, bone2Name=Head, swing1LimitDegrees=45, swing2LimitDegrees=45, twistLimitDegrees=45}
 {bone1Name=Neck_Armor_r, bone2Name=Neck_01, swing1LimitDegrees=45, swing2LimitDegrees=45, twistLimitDegrees=45}
 {bone1Name=Wing_shoulder_l, bone2Name=Body_Offset, swing1LimitDegrees=45, swing2LimitDegrees=45, twistLimitDegrees=45}
 {bone1Name=Wing_shoulder_l1, bone2Name=Wing_shoulder_l, swing1LimitDegrees=45, swing2LimitDegrees=45, twistLimitDegrees=45}
 {bone1Name=Wing_shoulder_l2, bone2Name=Wing_shoulder_l1, swing1LimitDegrees=45, swing2LimitDegrees=45, twistLimitDegrees=45}
 {bone1Name=Wing_shoulder_l3, bone2Name=Wing_shoulder_l2, swing1LimitDegrees=45, swing2LimitDegrees=45, twistLimitDegrees=45}
 {bone1Name=Feather_1_l, bone2Name=Wing_shoulder_l3, swing1LimitDegrees=45, swing2LimitDegrees=45, twistLimitDegrees=45}
 {bone1Name=Feather_2_l, bone2Name=Wing_shoulder_l3, swing1LimitDegrees=45, swing2LimitDegrees=45, twistLimitDegrees=45}
 {bone1Name=Feather_3_l, bone2Name=Wing_shoulder_l3, swing1LimitDegrees=45, swing2LimitDegrees=45, twistLimitDegrees=45}
 {bone1Name=Wing_shoulder_r, bone2Name=Body_Offset, swing1LimitDegrees=45, swing2LimitDegrees=45, twistLimitDegrees=45}
 {bone1Name=Wing_shoulder_r1, bone2Name=Wing_shoulder_r, swing1LimitDegrees=45, swing2LimitDegrees=45, twistLimitDegrees=45}
 {bone1Name=Wing_shoulder_r2, bone2Name=Wing_shoulder_r1, swing1LimitDegrees=45, swing2LimitDegrees=45, twistLimitDegrees=45}
 {bone1Name=Wing_shoulder_r3, bone2Name=Wing_shoulder_r2, swing1LimitDegrees=45, swing2LimitDegrees=45, twistLimitDegrees=45}
 {bone1Name=Feather_1_r, bone2Name=Wing_shoulder_r3, swing1LimitDegrees=45, swing2LimitDegrees=45, twistLimitDegrees=45}
 {bone1Name=Feather_2_r, bone2Name=Wing_shoulder_r3, swing1LimitDegrees=45, swing2LimitDegrees=45, twistLimitDegrees=45}
 {bone1Name=Feather_3_r, bone2Name=Wing_shoulder_r3, swing1LimitDegrees=45, swing2LimitDegrees=45, twistLimitDegrees=45}
 {bone1Name=Tail, bone2Name=Body_Offset, swing1LimitDegrees=45, swing2LimitDegrees=45, twistLimitDegrees=45}
 {bone1Name=Tail1, bone2Name=Tail, swing1LimitDegrees=45, swing2LimitDegrees=45, twistLimitDegrees=45}
 {bone1Name=Tail_wing_l_1, bone2Name=Tail1, swing1LimitDegrees=45, swing2LimitDegrees=45, twistLimitDegrees=45}
 {bone1Name=Tail_wing_l_2, bone2Name=Tail1, swing1LimitDegrees=45, swing2LimitDegrees=45, twistLimitDegrees=45}
 {bone1Name=Tail_wing_r_1, bone2Name=Tail1, swing1LimitDegrees=45, swing2LimitDegrees=45, twistLimitDegrees=45}
 {bone1Name=Tail_wing_r_2, bone2Name=Tail1, swing1LimitDegrees=45, swing2LimitDegrees=45, twistLimitDegrees=45}
 {bone1Name=Thigh_l, bone2Name=Body_Offset, swing1LimitDegrees=45, swing2LimitDegrees=45, twistLimitDegrees=45}
 {bone1Name=Calf_l, bone2Name=Thigh_l, swing1LimitDegrees=45, swing2LimitDegrees=45, twistLimitDegrees=45}
 {bone1Name=Claw_front_l_m_1, bone2Name=Calf_l, swing1LimitDegrees=45, swing2LimitDegrees=45, twistLimitDegrees=45}
 {bone1Name=Claw_front_l_r_1, bone2Name=Calf_l, swing1LimitDegrees=45, swing2LimitDegrees=45, twistLimitDegrees=45}
 {bone1Name=Thigh_r, bone2Name=Body_Offset, swing1LimitDegrees=45, swing2LimitDegrees=45, twistLimitDegrees=45}
 {bone1Name=Calf_r, bone2Name=Thigh_r, swing1LimitDegrees=45, swing2LimitDegrees=45, twistLimitDegrees=45}
 {bone1Name=Claw_front_r_1, bone2Name=Calf_r, swing1LimitDegrees=45, swing2LimitDegrees=45, twistLimitDegrees=45}
]
